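# Item Shop tag callout widget, built on AthenaItemShopTagRotator, that cycles tag text in a slanted pill with a plus icon.
# OnActiveStatusChanged switches between the FullInfo pill and a plus-only Preview; intro and outro animations, triggered by interaction events or handler functions, reveal each respectively.
# OnTagExit slides text out through a clipped overlay, then calls UpdateTagText when finished.
# OnTagEnter slides text in with a material border flash, burst and icon pulse.
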WidgetBlueprint /Game/UI/Frontend/ItemShop/FortTagRotator
animations [
 Out {
  binding_count=1
  length_seconds=0.35
  object=&self:Out
  sequence {bindings [Text_TagCallout {
     id=#1
     possessed_class=CommonTextBlock
     tracks [{
       type=MovieScene2DTransformTrack
       sections [{
         type=MovieScene2DTransformSection
         end_frame=7
         start_frame=0
         channels [Translation.X {type=MovieSceneScriptingFloatChannel, keys [[0,0],[7,144]]}]
        }]
      }]
    }]}
 }
 In {
  binding_count=4
  length_seconds=1
  object=&self:In
  sequence {bindings [
    Text_TagCallout {
     id=#1
     possessed_class=CommonTextBlock
     tracks [{
       type=MovieScene2DTransformTrack
       sections [{type=MovieScene2DTransformSection, channels [Translation.X {type=MovieSceneScriptingFloatChannel, keys [[0,-140],[6,0]]}]}]
      }]
    }
    Backing {
     id=#2
     possessed_class=Image
     tracks [{
       type=MovieSceneWidgetMaterialTrack
       sections [{
         type=MovieSceneParameterSection
         end_frame=20
         start_frame=0
         channels [Animate_BorderColor_Hover {type=MovieSceneScriptingFloatChannel, keys [[0,0],[2,1],[20.0003,0]]}]
        }]
      }]
    }
    Burst {
     id=#3
     possessed_class=Image
     tracks [
      {
       type=MovieSceneFloatTrack
       display_name=Render Opacity
       sections [{type=MovieSceneFloatSection, channels [None {type=MovieSceneScriptingFloatChannel, keys [[0,0],[2,0.25],[20.0003,0]]}]}]
      }
      {
       type=MovieScene2DTransformTrack
       sections [{
         type=MovieScene2DTransformSection
         channels [
          Scale.X {type=MovieSceneScriptingFloatChannel, keys [[0,1],[20,1.15]]}
          Scale.Y {type=MovieSceneScriptingFloatChannel, keys [[0,1],[20,1.75]]}
         ]
        }]
      }
     ]
    }
    PlusIconSmall {
     id=#4
     possessed_class=SizeBox
     tracks [{
       type=MovieScene2DTransformTrack
       sections [{
         type=MovieScene2DTransformSection
         channels [
          Translation.X {type=MovieSceneScriptingFloatChannel}
          Translation.Y {type=MovieSceneScriptingFloatChannel}
          Angle {type=MovieSceneScriptingFloatChannel}
          Scale.X {type=MovieSceneScriptingFloatChannel, keys [[0,1],[2,1.175],[6,1]]}
          Scale.Y {type=MovieSceneScriptingFloatChannel, keys [[0,1],[2,1.175],[6,1]]}
          Shear.X {type=MovieSceneScriptingFloatChannel}
          Shear.Y {type=MovieSceneScriptingFloatChannel}
         ]
        }]
      }]
    }
   ]}
 }
 IntroWidget {
  binding_count=1
  length_seconds=0.2
  object=&self:IntroWidget
  sequence {bindings [FullInfo {
     id=#5
     possessed_class=Overlay
     tracks [
      {
       type=MovieSceneFloatTrack
       display_name=Render Opacity
       sections [{type=MovieSceneFloatSection, channels [None {type=MovieSceneScriptingFloatChannel, keys [[0,0],[2,1]]}]}]
      }
      {
       type=MovieScene2DTransformTrack
       sections [{
         type=MovieScene2DTransformSection
         end_frame=4
         start_frame=0
         channels [Translation.X {type=MovieSceneScriptingFloatChannel, keys [[0,-64],[4,0]]}]
        }]
      }
     ]
    }]}
 }
 OutroWidget {
  binding_count=1
  length_seconds=0.2
  object=&self:OutroWidget
  sequence {bindings [Preview {
     id=#6
     possessed_class=Overlay
     tracks [
      {
       type=MovieSceneFloatTrack
       display_name=Render Opacity
       sections [{type=MovieSceneFloatSection, channels [None {type=MovieSceneScriptingFloatChannel, keys [[0,0],[1,1]]}]}]
      }
      {
       type=MovieScene2DTransformTrack
       sections [{
         type=MovieScene2DTransformSection
         channels [
          Translation.X {type=MovieSceneScriptingFloatChannel}
          Translation.Y {type=MovieSceneScriptingFloatChannel, keys [[0,16],[4,0]]}
          Scale.X {type=MovieSceneScriptingFloatChannel, keys [[0,0.75],[4,1]]}
          Scale.Y {type=MovieSceneScriptingFloatChannel, keys [[0,0.75],[4,1]]}
         ]
        }]
      }
     ]
    }]}
 }
]
graphs [
 Cosmetic {nodes [
   OnStartedInteracting() {is=title K2Node_CustomEvent}
   > PlayAnimationForward(InAnimation=Get IntroWidget() -> (IntroWidget:Widget Animation Object Reference))
   OnFinishedInteracting() {is=title K2Node_CustomEvent}
   > PlayAnimationForward(InAnimation=Get OutroWidget() -> (OutroWidget:Widget Animation Object Reference))
   n1 = Event OnActiveStatusChanged!() -> (OutputDelegate, then, bIsActive:bool) {is=title K2Node_Event}
   > SetActiveWidget(
    self=Get StatusSwitcher() -> (StatusSwitcher:Common Visibility Switcher Object Reference)
    Widget=Select(
     Option 0:Widget Object Reference=Get Preview() -> (Preview:Overlay Object Reference)
     Option 1:Widget Object Reference=Get FullInfo() -> (FullInfo:Overlay Object Reference)
     Index=@n1.bIsActive
    ) -> (ReturnValue:Widget Object Reference)
   )
   Event OnTagEnter() {is=title K2Node_Event}
   > "Play Animation with Finished event (legacy)"!(
    execute:Exec
    Widget:User Widget Object Reference=Self-Reference()
    InAnimation:Widget Animation Object Reference=Get In() -> (In:Widget Animation Object Reference)
    StartAtTime:float=0
    NumLoopsToPlay:int=1
    PlayMode:EUMGSequencePlayMode Enum
    PlaybackSpeed:float=1
   ) -> (then:Exec, Result:UMGSequence Player Object Reference, Finished:Exec) {is=title K2Node_PlayAnimation}
   Event OnTagExit() {is=title K2Node_Event}
   > n2 = "Play Animation with Finished event (legacy)"(Widget=Self-Reference(), InAnimation=Get Out() -> (Out:Widget Animation Object Reference), StartAtTime=0, NumLoopsToPlay=1, PlaybackSpeed=1)
   UpdateTagText(execute=@n2.Finished, self:Athena Item Shop Tag Rotator Object Reference)
  ]}
 EventGraph {nodes [Event Construct()]}
 HandleIntroAnim {nodes [
   HandleIntroAnim() {is=title K2Node_FunctionEntry}
   > PlayAnimationForward(InAnimation=Get IntroWidget())
  ]}
 HandleOutroAnim {nodes [
   HandleOutroAnim() {is=title K2Node_FunctionEntry}
   > PlayAnimationForward(InAnimation=Get OutroWidget())
  ]}
]
variables []
widgets {
 info {inheritedWidgetCount=1, parentClass=/Script/FortniteUI.AthenaItemShopTagRotator, rootWidgetClass=&Overlay, widgetCount=16}
 widgets [
  Overlay_18 {widgetClassPath=&Overlay, widget=Overlay_18}
  SizeBox_1 {
   widgetClassPath=&SizeBox
   parent=Overlay_18
   widget=SizeBox_1
   properties {bOverride_HeightOverride=true, heightOverride=40, widthOverride=168}
   slot {class=OverlaySlot, properties {horizontalAlignment=HAlign_Fill, verticalAlignment=VAlign_Center}}
  }
  StatusSwitcher {
   widgetClassPath=&CommonVisibilitySwitcher
   bIsVariable=true
   parent=SizeBox_1
   widget=StatusSwitcher
   properties {activeWidgetIndex=1}
   slot {class=SizeBoxSlot}
  }
  FullInfo {widgetClassPath=&Overlay, bIsVariable=true, parent=StatusSwitcher, widget=FullInfo, slot {class=CommonVisibilitySwitcherSlot}}
  Overlay_0 {
   widgetClassPath=&Overlay
   parent=FullInfo
   widget=Overlay_0
   slot {class=OverlaySlot, properties {horizontalAlignment=HAlign_Fill, verticalAlignment=VAlign_Fill}}
  }
  Shadow {
   widgetClassPath=&Image
   bIsVariable=true
   parent=Overlay_0
   widget=Shadow
   properties {
    brush {
     bIsDynamicallyLoaded=false
     drawAs=Box
     imageType=NoImage
     mirroring=NoMirror
     resourceName=None
     resourceObject=/Game/UI/Frontend/ItemShop/Art/M_UI_OfferTile_TagBacking_Shadow
     tiling=NoTile
     imageSize=<40 40>
     margin=<ltrb 0.5 0 0.5 0>
     outlineSettings {
      bUseBrushTransparency=false
      roundingType=HalfHeightRadius
      width=0
      color {colorUseRule=UseColor_Specified, specifiedColor=<rgba 0 0 0 0>}
      cornerRadii=<0 0 0 0>
     }
     tintColor {colorUseRule=UseColor_Specified, specifiedColor=<rgba 1 1 1 1>}
     uVRegion {bIsValid=false, max=<0 0>, min=<0 0>}
    }
    renderTransform {angle=0, scale=<1.15 1.75>, shear=<0 0>, translation=<0 0>}
   }
   slot {class=OverlaySlot, properties {horizontalAlignment=HAlign_Fill, verticalAlignment=VAlign_Fill}}
  }
  Burst {
   widgetClassPath=&Image
   bIsVariable=true
   parent=Overlay_0
   widget=Burst
   properties {
    renderOpacity=0
    brush {
     bIsDynamicallyLoaded=false
     drawAs=Box
     imageType=NoImage
     mirroring=NoMirror
     resourceName=None
     resourceObject=/Game/UI/Frontend/ItemShop/Art/M_UI_OfferTile_TagBacking_Burst
     tiling=NoTile
     imageSize=<40 40>
     margin=<ltrb 0.5 0 0.5 0>
     outlineSettings {
      bUseBrushTransparency=false
      roundingType=HalfHeightRadius
      width=0
      color {colorUseRule=UseColor_Specified, specifiedColor=<rgba 0 0 0 0>}
      cornerRadii=<0 0 0 0>
     }
     tintColor {colorUseRule=UseColor_Specified, specifiedColor=<rgba 1 1 1 1>}
     uVRegion {bIsValid=false, max=<0 0>, min=<0 0>}
    }
   }
   slot {class=OverlaySlot, properties {horizontalAlignment=HAlign_Fill, verticalAlignment=VAlign_Fill}}
  }
  Backing {
   widgetClassPath=&Image
   bIsVariable=true
   parent=Overlay_0
   widget=Backing
   properties {brush {
     bIsDynamicallyLoaded=false
     drawAs=Box
     imageType=NoImage
     mirroring=NoMirror
     resourceName=None
     resourceObject=/Game/UI/Frontend/ItemShop/Art/M_UI_OfferTile_TagBacking
     tiling=NoTile
     imageSize=<40 40>
     margin=<ltrb 0.5 0 0.5 0>
     outlineSettings {
      bUseBrushTransparency=false
      roundingType=HalfHeightRadius
      width=0
      color {colorUseRule=UseColor_Specified, specifiedColor=<rgba 0 0 0 0>}
      cornerRadii=<0 0 0 0>
     }
     tintColor {colorUseRule=UseColor_Specified, specifiedColor=<rgba 1 1 1 1>}
     uVRegion {bIsValid=false, max=<0 0>, min=<0 0>}
    }}
   slot {class=OverlaySlot, properties {horizontalAlignment=HAlign_Fill, verticalAlignment=VAlign_Fill}}
  }
  HorizontalBox_0 {
   widgetClassPath=&HorizontalBox
   parent=FullInfo
   widget=HorizontalBox_0
   properties {renderTransform {angle=-1.5, scale=<1 1>, shear=<0 0>, translation=<0 0>}}
   slot {class=OverlaySlot, properties {horizontalAlignment=HAlign_Center, verticalAlignment=VAlign_Center, padding=<ltrb 16 0 16 0>}}
  }
  PlusIconSmall {
   widgetClassPath=&SizeBox
   parent=HorizontalBox_0
   widget=PlusIconSmall
   properties {bOverride_HeightOverride=true, bOverride_WidthOverride=true, heightOverride=20, widthOverride=20}
   slot {class=HorizontalBoxSlot, properties {horizontalAlignment=HAlign_Center, verticalAlignment=VAlign_Center, padding=<ltrb 0 0 2 0>}}
  }
  Image_164 {
   widgetClassPath=&Image
   bIsVariable=true
   parent=PlusIconSmall
   widget=Image_164
   properties {
    brush {
     bIsDynamicallyLoaded=false
     drawAs=Image
     imageType=NoImage
     mirroring=NoMirror
     resourceName=None
     resourceObject=/Game/Athena/UI/Frontend/Parties/Art/T_UI_PlusSign_64x
     tiling=NoTile
     imageSize=<64 64>
     margin=<ltrb 0 0 0 0>
     outlineSettings {
      bUseBrushTransparency=false
      roundingType=HalfHeightRadius
      width=0
      color {colorUseRule=UseColor_Specified, specifiedColor=<rgba 0 0 0 0>}
      cornerRadii=<0 0 0 0>
     }
     tintColor {colorUseRule=UseColor_Specified, specifiedColor=<rgba 1 1 1 1>}
     uVRegion {bIsValid=false, max=<0 0>, min=<0 0>}
    }
    renderTransform {angle=0, scale=<1 1>, shear=<-6 0>, translation=<0 0>}
   }
   slot {class=SizeBoxSlot}
  }
  Overlay_2 {
   widgetClassPath=&Overlay
   parent=HorizontalBox_0
   widget=Overlay_2
   properties {clipping=ClipToBounds, renderTransform {angle=0, scale=<1 1>, shear=<0 0>, translation=<0 1>}}
   slot {class=HorizontalBoxSlot, properties {padding=<ltrb 2 0 0 0>}}
  }
  Text_TagCallout {
   widgetClassPath=&CommonTextBlock
   bInherited=true
   parent=Overlay_2
   widget=Text_TagCallout
   properties {
    justification=Center
    mobileFontSizeMultiplier=1.34
    text=Builtin Emote
    textTransformPolicy=ToUpper
    font {
     bForceMonospaced=false
     bMaterialIsStencil=false
     fontMaterial=null
     fontObject=&BurbankSmall
     letterSpacing=0
     monospacedWidth=1
     size=13
     skewAmount=0.2
     typefaceFontName=Black
     outlineSettings {
      bApplyOutlineToDropShadows=false
      bMiteredCorners=false
      bSeparateFillAlpha=false
      outlineMaterial=null
      outlineSize=2
      outlineColor=<rgba 0 0 0 0>
     }
    }
    shadowOffset=<0 0>
   }
   slot {class=OverlaySlot, properties {horizontalAlignment=HAlign_Fill, verticalAlignment=VAlign_Fill}}
  }
  Preview {
   widgetClassPath=&Overlay
   bIsVariable=true
   parent=StatusSwitcher
   widget=Preview
   slot {class=CommonVisibilitySwitcherSlot, properties {horizontalAlignment=HAlign_Right, verticalAlignment=VAlign_Bottom}}
  }
  SizeBox_3 {
   widgetClassPath=&SizeBox
   parent=Preview
   widget=SizeBox_3
   properties {
    bOverride_HeightOverride=true
    bOverride_WidthOverride=true
    heightOverride=40
    widthOverride=40
    renderTransform {angle=0, scale=<1 1>, shear=<-12 -4>, translation=<0 0>}
   }
   slot {class=OverlaySlot}
  }
  Image {
   widgetClassPath=&Image
   bIsVariable=true
   parent=SizeBox_3
   widget=Image
   properties {brush {
     bIsDynamicallyLoaded=false
     drawAs=Image
     imageType=NoImage
     mirroring=NoMirror
     resourceName=None
     resourceObject=&T_UI_PlusSign_64x
     tiling=NoTile
     imageSize=<64 64>
     margin=<ltrb 0 0 0 0>
     outlineSettings {
      bUseBrushTransparency=false
      roundingType=HalfHeightRadius
      width=0
      color {colorUseRule=UseColor_Specified, specifiedColor=<rgba 0 0 0 0>}
      cornerRadii=<0 0 0 0>
     }
     tintColor {colorUseRule=UseColor_Specified, specifiedColor=<rgba 1 1 1 1>}
     uVRegion {bIsValid=false, max=<0 0>, min=<0 0>}
    }}
   slot {class=SizeBoxSlot}
  }
 ]
}
