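# UI-domain material for the Rumble item shop background, layering animated orange-yellow smoke over a dark radial black-to-blue gradient.
# Panned movement and blurred-noise textures form the smoke, tinted by a radial orange-to-red blend and masked by vertical gradients.
# MF_UI_Confetti adds a masked orange particle layer, blended with the smoke and composited through a yellow vertical gradient.
# Named reroutes declare aspect-corrected UVs, gradient masks and scrolling-noise UV distortion.
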Material /ItemShop/UI/Art/Rumble/M_UI_Rumble_ShopBG_Smoke
graph {
 properties {bUsedWithStaticMesh=false, materialDomain=MD_UI, pixelDepthOffsetMode=PDOM_Legacy, usedShadingModels=Default Lit}
 nodes [
  Add()
  n1 = NamedRerouteUsage()
  n2 = AppendVector(A=ScalarParameter() {defaultValue=0.7793, parameterName=Radial-X-Tile}, B=ScalarParameter() {defaultValue=1, parameterName=Radial-Y-Tile})
  n3 = Constant2Vector() {g=1, r=0.5}
  n4 = Power(Base=Distance(A=Multiply(A=@n1, B=@n2), B=Multiply(A=@n2, B=@n3)), Exp=ScalarParameter() {defaultValue=2.5, parameterName=Distance Power})
  n5 = ScalarParameter() {parameterName=L1-Patter-X-Offset}
  n6 = ScalarParameter() {defaultValue=0.172, parameterName=L1-Patter-Y-Offset}
  n7 = MaterialFunctionCall!(
   "UV(V2)"=Add(A=NamedRerouteUsage(), B=AppendVector(A=@n5, B=@n6))
   Global-Angle=ScalarParameter() {defaultValue=-1, parameterName=L1-Global Angle}
   "Pattern-Tiles(V2)"=AppendVector(A=ScalarParameter() {defaultValue=50, parameterName=L1-Pattern-X}, B=ScalarParameter() {defaultValue=0.75, parameterName=L1-Pattern-Y})
   RandomSeed=ScalarParameter() {defaultValue=12345, parameterName=L1-RandomSeed}
   Confetti-Speed-TimeFactor=ScalarParameter() {defaultValue=0.138, parameterName=L1-Confetti-Speed-TimeFactor}
   Confetti-Rotation-TimeFactor=ScalarParameter() {defaultValue=0.794, parameterName=L1-Confetti-Rotation-TimeFactor}
   Confetti-Local-Rotation-Range=ScalarParameter() {defaultValue=0.0075, parameterName=L1-Confetti-Local-Rotation-Range}
   "Confetti-Local-Tiles(V2)"=AppendVector(
    A=ScalarParameter() {defaultValue=5.2803, parameterName=L1-Confetti-Local-Tile-X}
    B=ScalarParameter() {defaultValue=331.8715, parameterName=L1-Confetti-Local-Tile-Y}
   )
   "SmoothStepRanges(V2)"=AppendVector(
    A=ScalarParameter() {defaultValue=0.6489, parameterName=L1-SmoothStep-Max}
    B=ScalarParameter() {defaultValue=0.5, parameterName=L1-SmoothStep-Min}
   )
  ) {materialFunction=/Game/UI/Frontend/ItemShop/Art/Cactus/MF_UI_Confetti}
  n8 = ComponentMask(None=TextureCoordinate()) {g=true}
  n9 = ScalarParameter() {defaultValue=0.236, parameterName=L1-Gradient-Mask}
  n10 = Saturate(None=Multiply(A=Subtract(A=@n8, B=@n9), B=ScalarParameter() {defaultValue=2, parameterName=L1-Gradient-Falloff}))
  n11 = Multiply(
   A=Constant3Vector() {constant=<rgba 1 0.3366 0.0015 1>}
   B=Multiply(A=Multiply(A=@n7.Confetti-Mask, B=@n10), B=ScalarParameter() {defaultValue=1, parameterName=L1-Intensity})
  )
  n12 = Constant3Vector() {constant=<rgba 1 0.2274 0 0>}
  n13 = Constant3Vector() {constant=<rgba 0.3646 0.0186 0 0>}
  n14 = AppendVector(
   A=ScalarParameter() {defaultValue=0.9863, parameterName=Radial-X-Tile2}
   B=ScalarParameter() {defaultValue=1.3969, parameterName=Radial-Y-Tile2}
  )
  n15 = Multiply(A=NamedRerouteUsage(), B=@n14)
  n16 = Multiply(A=@n14, B=Constant2Vector() {g=1, r=0.5})
  n17 = Saturate(None=Power(Base=Distance(A=@n15, B=@n16), Exp=ScalarParameter() {defaultValue=0.5829, parameterName=Distance Powerfire}))
  n18 = NamedRerouteUsage()
  n19 = NamedRerouteUsage()
  n20 = Time()
  n21 = ScalarParameter() {defaultValue=0.75, parameterName=Speed}
  n22 = TextureSample(UVs=Panner(Coordinate=Add(A=@n18, B=@n19), Time=Multiply(A=@n20, B=@n21)) {speedY=0.25}) {texture=/ItemShop/UI/Art/Cactus/Textures/T_UI_CactusMovement}
  n23 = NamedRerouteUsage()
  n24 = NamedRerouteUsage()
  n25 = Time()
  n26 = ScalarParameter() {defaultValue=0.25, parameterName=Speed2}
  n27 = TextureSample(UVs=Panner(Coordinate=Add(A=@n23, B=@n24), Time=Multiply(A=@n25, B=@n26)) {speedY=0.25}) {samplerType=SAMPLERTYPE_Grayscale, texture=/Game/UI/Frontend/ItemShop/Art/Rumble/LowResBlurredNoise_Rumble}
  n28 = ComponentMask(None=TextureCoordinate()) {g=true}
  n29 = ScalarParameter() {defaultValue=-0.067, parameterName=Subtract3}
  n30 = Multiply(
   A=Multiply(A=Subtract(A=@n28, B=@n29), B=ScalarParameter() {defaultValue=-0.0042, parameterName=DropOff3})
   B=ScalarParameter() {defaultValue=7.6175, parameterName=Intensity3}
  )
  n31 = NamedRerouteUsage()
  n32 = NamedRerouteUsage()
  n33 = Time()
  n34 = ScalarParameter() {defaultValue=1, parameterName=TopMaskScale}
  n35 = TextureSample(UVs=Panner(Coordinate=Add(A=@n31, B=@n32), Time=Multiply(A=@n33, B=@n34)) {speedX=0.015, speedY=0.15}) {texture=&T_UI_CactusMovement}
  n36 = LinearInterpolate(
   B=Multiply(A=LinearInterpolate(A=@n12, B=@n13, Alpha=@n17), B=Add(A=@n22.RGB, B=@n27.RGB))
   Alpha=LinearInterpolate(A=Add(A=@n30, B=@n35.RGB), Alpha=NamedRerouteUsage())
  )
  n37 = Subtract(A=ComponentMask(None=TextureCoordinate()) {g=true}, B=ScalarParameter() {defaultValue=0.433, parameterName=SubtractYellow})
  n38 = ScalarParameter() {defaultValue=2.4192, parameterName=FalloffYellow}
  n39 = LinearInterpolate(B=Constant3Vector() {constant=<rgba 1 0.6749 0 0>}, Alpha=Saturate(None=Multiply(A=@n37, B=@n38)))
  LinearInterpolate(
   A=LinearInterpolate(A=Constant3Vector(), B=Constant3Vector() {constant=<rgba 0 0.0115 0.2656 0>}, Alpha=Saturate(None=@n4))
   B=Multiply(A=LinearInterpolate(A=@n11, B=@n36), B=@n39)
   Alpha=@n39
  )
  n40 = MaterialFunctionCall() {materialFunction=&GetUserInterfaceUV}
  n41 = MaterialFunctionCall(Float2=@n40.Pixel Size) {materialFunction=&BreakOutFloat2Components}
  n42 = AppendVector(A=Divide(A=@n41.R, B=Max(A=@n41.G)), B=Constant() {r=1})
  n43 = OneMinus(None=@n42)
  NamedRerouteDeclaration(None=Add(A=Multiply(A=@n43) {constB=0.5}, B=Multiply(A=@n42, B=@n40.Normalized UV))) {name="Squared UV - (x/y,y)", nodeColor=<rgba 0 0 0 1>}
  n44 = ComponentMask(None=TextureCoordinate()) {g=true}
  n45 = ScalarParameter() {defaultValue=0.3119, parameterName=Subtract}
  NamedRerouteDeclaration(None=Multiply(A=Subtract(A=@n44, B=@n45), B=ScalarParameter() {defaultValue=0.4349, parameterName=DropOff})) {name=GradientMask, nodeColor=<rgba 0.2188 0 1 1>}
  n46 = Panner(
   Coordinate=Multiply(A=TextureCoordinate(), B=ScalarParameter() {defaultValue=0.75, parameterName=DupeScaler})
   Time=Multiply(A=Time(), B=ScalarParameter() {defaultValue=1.5, parameterName=TimeFactor})
  ) {speedX=0.058, speedY=0.096}
  NamedRerouteDeclaration(None=Multiply(
    A=TextureSample(UVs=@n46) {samplerType=SAMPLERTYPE_Grayscale, texture=&LowResBlurredNoise}.R
    B=ScalarParameter() {defaultValue=0.15, parameterName=DistortionScaler}
   )) {name=UV-Distortion, nodeColor=<rgba 0 0 0 1>}
  n47 = Subtract(A=ComponentMask(None=TextureCoordinate()) {g=true}, B=ScalarParameter() {defaultValue=0.608, parameterName=Subtract2})
  n48 = ScalarParameter() {defaultValue=0.6824, parameterName=DropOff2}
  NamedRerouteDeclaration(None=Multiply(A=Multiply(A=@n47, B=@n48), B=ScalarParameter() {defaultValue=0.984, parameterName=Intensity2})) {name=GradientMask_2, nodeColor=<rgba 0 0 0 1>}
  NamedRerouteUsage()
  OneMinus(None=@n8)
  Power(Exp=ScalarParameter() {defaultValue=0.6825, parameterName=ConfettiParam2})
 ]
}
