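# Debug test widget that previews Item Shop offer tiles in large, normal and small sizes side by side.
# Three stacked full-screen background images sit behind the tiles, including a special Rumble smoke background.
# Each duplicated tile layers a placeholder offer image, tag rotator, name panel and a visibility switcher between price and Owned states.
# The layout is static mockup content; its event graph holds only unconnected PreConstruct, Construct and Tick events.
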WidgetBlueprint /Game/NeverCook/UI/Debug/TestWidgets/ItemShop/TestWBP_ItemShop
graphs [EventGraph {nodes [
   Event PreConstruct()
   Event Construct()
   Event Tick()
  ]}]
variables []
widgets {
 info {inheritedWidgetCount=0, parentClass=&UserWidget, rootWidgetClass=&Overlay, widgetCount=69}
 widgets [
  Overlay_0 {widgetClassPath=&Overlay, widget=Overlay_0}
  Image_DynamicBackground {
   widgetClassPath=&Image
   bIsVariable=true
   parent=Overlay_0
   widget=Image_DynamicBackground
   properties {
    visibility=HitTestInvisible
    brush {
     bIsDynamicallyLoaded=false
     drawAs=Image
     imageType=NoImage
     mirroring=NoMirror
     resourceName=None
     resourceObject=/FortniteCoreUI/Frontend/Art/M_UI_RadialGradient_ItemShop_BG
     tiling=NoTile
     imageSize=<32 32>
     margin=<ltrb 0 0 0 0>
     outlineSettings {
      bUseBrushTransparency=false
      roundingType=HalfHeightRadius
      width=0
      color {colorUseRule=UseColor_Specified, specifiedColor=<rgba 0 0 0 0>}
      cornerRadii=<0 0 0 0>
     }
     tintColor {colorUseRule=UseColor_Specified, specifiedColor=<rgba 1 1 1 1>}
     uVRegion {bIsValid=false, max=<0 0>, min=<0 0>}
    }
   }
   slot {class=OverlaySlot, properties {horizontalAlignment=HAlign_Fill, verticalAlignment=VAlign_Fill}}
  }
  Image_DefaultBackground {
   widgetClassPath=&Image
   bIsVariable=true
   parent=Overlay_0
   widget=Image_DefaultBackground
   properties {
    visibility=HitTestInvisible
    brush {
     bIsDynamicallyLoaded=false
     drawAs=Image
     imageType=NoImage
     mirroring=NoMirror
     resourceName=None
     resourceObject=&M_UI_RadialGradient_ItemShop_BG
     tiling=NoTile
     imageSize=<32 32>
     margin=<ltrb 0 0 0 0>
     outlineSettings {
      bUseBrushTransparency=false
      roundingType=HalfHeightRadius
      width=0
      color {colorUseRule=UseColor_Specified, specifiedColor=<rgba 0 0 0 0>}
      cornerRadii=<0 0 0 0>
     }
     tintColor {colorUseRule=UseColor_Specified, specifiedColor=<rgba 1 1 1 1>}
     uVRegion {bIsValid=false, max=<0 0>, min=<0 0>}
    }
   }
   slot {class=OverlaySlot, properties {horizontalAlignment=HAlign_Fill, verticalAlignment=VAlign_Fill}}
  }
  Image_SpecialBackground {
   widgetClassPath=&Image
   bIsVariable=true
   parent=Overlay_0
   widget=Image_SpecialBackground
   properties {
    visibility=HitTestInvisible
    brush {
     bIsDynamicallyLoaded=false
     drawAs=Image
     imageType=NoImage
     mirroring=NoMirror
     resourceName=None
     resourceObject=/ItemShop/UI/Art/Rumble/M_UI_Rumble_ShopBG_Smoke
     tiling=NoTile
     imageSize=<32 32>
     margin=<ltrb 0 0 0 0>
     outlineSettings {
      bUseBrushTransparency=false
      roundingType=HalfHeightRadius
      width=0
      color {colorUseRule=UseColor_Specified, specifiedColor=<rgba 0 0 0 0>}
      cornerRadii=<0 0 0 0>
     }
     tintColor {colorUseRule=UseColor_Specified, specifiedColor=<rgba 1 1 1 1>}
     uVRegion {bIsValid=false, max=<0 0>, min=<0 0>}
    }
   }
   slot {class=OverlaySlot, properties {horizontalAlignment=HAlign_Fill, verticalAlignment=VAlign_Fill}}
  }
  Overlay_84 {
   widgetClassPath=&Overlay
   parent=Overlay_0
   widget=Overlay_84
   slot {class=OverlaySlot, properties {horizontalAlignment=HAlign_Center, verticalAlignment=VAlign_Center}}
  }
  WithUI {
   widgetClassPath=&HorizontalBox
   parent=Overlay_84
   widget=WithUI
   slot {class=OverlaySlot, properties {horizontalAlignment=HAlign_Center, verticalAlignment=VAlign_Center}}
  }
  Large_UI {
   widgetClassPath=&SizeBox
   parent=WithUI
   widget=Large_UI
   properties {bOverride_HeightOverride=true, bOverride_WidthOverride=true, heightOverride=720, widthOverride=864}
   slot {class=HorizontalBoxSlot}
  }
  LargeOverlay {widgetClassPath=&Overlay, parent=Large_UI, widget=LargeOverlay, slot {class=SizeBoxSlot}}
  Image_Offer_Large_UI {
   widgetClassPath=/Script/EpicCMSUIFramework.EpicCMSImage
   bIsVariable=true
   parent=LargeOverlay
   widget=Image_Offer_Large_UI
   properties {brush {
     bIsDynamicallyLoaded=false
     drawAs=Image
     imageType=NoImage
     mirroring=NoMirror
     resourceName=None
     resourceObject=/OfferCatalog/MI_OfferImages/MI_EID_GasStation
     tiling=NoTile
     imageSize=<512 512>
     margin=<ltrb 0 0 0 0>
     outlineSettings {
      bUseBrushTransparency=false
      roundingType=HalfHeightRadius
      width=0
      color {colorUseRule=UseColor_Specified, specifiedColor=<rgba 0 0 0 0>}
      cornerRadii=<0 0 0 0>
     }
     tintColor {colorUseRule=UseColor_Specified, specifiedColor=<rgba 1 1 1 1>}
     uVRegion {bIsValid=false, max=<0 0>, min=<0 0>}
    }}
   slot {class=OverlaySlot, properties {horizontalAlignment=HAlign_Fill, verticalAlignment=VAlign_Fill}}
  }
  UI_Layer {
   widgetClassPath=&VerticalBox
   parent=LargeOverlay
   widget=UI_Layer
   slot {class=OverlaySlot, properties {horizontalAlignment=HAlign_Fill, verticalAlignment=VAlign_Bottom}}
  }
  TagRotator_ItemTags {
   widgetClassPath=/Game/UI/Frontend/ItemShop/FortTagRotator
   bIsVariable=true
   parent=UI_Layer
   widget=TagRotator_ItemTags
   properties {timeBetweenTitleChanges=5}
   slot {class=VerticalBoxSlot, properties {horizontalAlignment=HAlign_Right, verticalAlignment=VAlign_Bottom, padding=<ltrb 0 0 12 0>}}
  }
  OfferIdentifier {
   widgetClassPath=&Overlay
   bIsVariable=true
   parent=UI_Layer
   widget=OfferIdentifier
   slot {class=VerticalBoxSlot, properties {padding=<ltrb 0 12 0 0>}}
  }
  UI_Panel_Backing_SizeBox {
   widgetClassPath=&SizeBox
   parent=OfferIdentifier
   widget=UI_Panel_Backing_SizeBox
   properties {bOverride_HeightOverride=true, heightOverride=108}
   slot {class=OverlaySlot, properties {horizontalAlignment=HAlign_Fill, verticalAlignment=VAlign_Fill}}
  }
  PanelBacking {
   widgetClassPath=&Image
   bIsVariable=true
   parent=UI_Panel_Backing_SizeBox
   widget=PanelBacking
   properties {brush {
     bIsDynamicallyLoaded=false
     drawAs=Image
     imageType=NoImage
     mirroring=NoMirror
     resourceName=None
     resourceObject=/OfferCatalog/Art/M_UI_OfferTile_UI_Layer
     tiling=NoTile
     imageSize=<32 32>
     margin=<ltrb 0 0 0 0>
     outlineSettings {
      bUseBrushTransparency=false
      roundingType=HalfHeightRadius
      width=0
      color {colorUseRule=UseColor_Specified, specifiedColor=<rgba 0 0 0 0>}
      cornerRadii=<0 0 0 0>
     }
     tintColor {colorUseRule=UseColor_Specified, specifiedColor=<rgba 1 1 1 1>}
     uVRegion {bIsValid=false, max=<0 0>, min=<0 0>}
    }}
   slot {class=SizeBoxSlot}
  }
  Text+Price {
   widgetClassPath=&VerticalBox
   parent=OfferIdentifier
   widget=Text+Price
   slot {class=OverlaySlot, properties {horizontalAlignment=HAlign_Fill, verticalAlignment=VAlign_Fill}}
  }
  Text {
   widgetClassPath=&VerticalBox
   parent=Text+Price
   widget=Text
   slot {class=VerticalBoxSlot, properties {verticalAlignment=VAlign_Center, padding=<ltrb 20 10 20 0>, size {sizeRule=Fill, value=1}}}
  }
  Text_Name {
   widgetClassPath=&CommonTextBlock
   bIsVariable=true
   parent=Text
   widget=Text_Name
   properties {
    justification=Center
    scrollStyle=&TextScrollStyle-Base_C
    text=Asset Name
    textTransformPolicy=ToUpper
    font {
     bForceMonospaced=false
     bMaterialIsStencil=false
     fontMaterial=/OfferCatalog/Art/M_UI_ItemShop_Text_Header
     fontObject=&Burbank
     letterSpacing=0
     monospacedWidth=1
     size=22
     skewAmount=0
     typefaceFontName=BurbankRegular-Black
     outlineSettings {
      bApplyOutlineToDropShadows=false
      bMiteredCorners=false
      bSeparateFillAlpha=false
      outlineMaterial=null
      outlineSize=0
      outlineColor=<rgba 0 0 0 1>
     }
    }
    renderTransform {angle=0, scale=<1 1>, shear=<-12 0>, translation=<0 0>}
    shadowOffset=<0 0>
   }
   slot {class=VerticalBoxSlot, properties {verticalAlignment=VAlign_Center, size {sizeRule=Fill, value=1}}}
  }
  OwnedStatus_SizeBox {
   widgetClassPath=&SizeBox
   parent=Text+Price
   widget=OwnedStatus_SizeBox
   properties {bOverride_HeightOverride=true, heightOverride=36}
   slot {class=VerticalBoxSlot, properties {verticalAlignment=VAlign_Bottom}}
  }
  Switcher_OwningStatus {
   widgetClassPath=&CommonVisibilitySwitcher
   bIsVariable=true
   parent=OwnedStatus_SizeBox
   widget=Switcher_OwningStatus
   slot {class=SizeBoxSlot}
  }
  TileState_NotOwned {
   widgetClassPath=&Overlay
   bIsVariable=true
   parent=Switcher_OwningStatus
   widget=TileState_NotOwned
   slot {class=CommonVisibilitySwitcherSlot}
  }
  Price_Backing {
   widgetClassPath=&Image
   bIsVariable=true
   parent=TileState_NotOwned
   widget=Price_Backing
   properties {brush {
     bIsDynamicallyLoaded=false
     drawAs=Box
     imageType=NoImage
     mirroring=NoMirror
     resourceName=None
     resourceObject=/OfferCatalog/Art/M_UI_OfferTile_StatusPanel
     tiling=NoTile
     imageSize=<32 32>
     margin=<ltrb 0.5 0 0.5 0>
     outlineSettings {
      bUseBrushTransparency=false
      roundingType=HalfHeightRadius
      width=0
      color {colorUseRule=UseColor_Specified, specifiedColor=<rgba 0 0 0 0>}
      cornerRadii=<0 0 0 0>
     }
     tintColor {colorUseRule=UseColor_Specified, specifiedColor=<rgba 1 1 1 1>}
     uVRegion {bIsValid=false, max=<0 0>, min=<0 0>}
    }}
   slot {class=OverlaySlot, properties {horizontalAlignment=HAlign_Fill, verticalAlignment=VAlign_Fill}}
  }
  HorizontalBox_0 {
   widgetClassPath=&HorizontalBox
   parent=TileState_NotOwned
   widget=HorizontalBox_0
   slot {class=OverlaySlot, properties {horizontalAlignment=HAlign_Right, verticalAlignment=VAlign_Center, padding=<ltrb 0 0 64 0>}}
  }
  Text_Price {
   widgetClassPath=&CommonTextBlock
   bIsVariable=true
   parent=HorizontalBox_0
   widget=Text_Price
   properties {
    justification=Center
    text=2000
    textTransformPolicy=ToUpper
    font {
     bForceMonospaced=false
     bMaterialIsStencil=false
     fontMaterial=/Game/UI/Frontend/ItemShop/Art/M_UI_ItemShop_Text_Price
     fontObject=&Burbank
     letterSpacing=0
     monospacedWidth=1
     size=16
     skewAmount=0
     typefaceFontName=BurbankRegular-Black
     outlineSettings {
      bApplyOutlineToDropShadows=false
      bMiteredCorners=false
      bSeparateFillAlpha=false
      outlineMaterial=null
      outlineSize=0
      outlineColor=<rgba 0 0 0 1>
     }
    }
    renderTransform {angle=0, scale=<1 1>, shear=<-12 0>, translation=<0 0>}
    shadowOffset=<0 0>
   }
   slot {class=HorizontalBoxSlot, properties {horizontalAlignment=HAlign_Right}}
  }
  TileState_Owned {
   widgetClassPath=&Overlay
   bIsVariable=true
   parent=Switcher_OwningStatus
   widget=TileState_Owned
   slot {class=CommonVisibilitySwitcherSlot}
  }
  OwnedBacking {
   widgetClassPath=&Image
   bIsVariable=true
   parent=TileState_Owned
   widget=OwnedBacking
   properties {brush {
     bIsDynamicallyLoaded=false
     drawAs=Image
     imageType=NoImage
     mirroring=NoMirror
     resourceName=None
     resourceObject=/OfferCatalog/Art/M_UI_OfferTile_StatusPanel_Owned
     tiling=NoTile
     imageSize=<32 32>
     margin=<ltrb 0 0 0 0>
     outlineSettings {
      bUseBrushTransparency=false
      roundingType=HalfHeightRadius
      width=0
      color {colorUseRule=UseColor_Specified, specifiedColor=<rgba 0 0 0 0>}
      cornerRadii=<0 0 0 0>
     }
     tintColor {colorUseRule=UseColor_Specified, specifiedColor=<rgba 1 1 1 1>}
     uVRegion {bIsValid=false, max=<0 0>, min=<0 0>}
    }}
   slot {class=OverlaySlot, properties {horizontalAlignment=HAlign_Fill}}
  }
  Text_Name_2 {
   widgetClassPath=&CommonTextBlock
   bIsVariable=true
   parent=TileState_Owned
   widget=Text_Name_2
   properties {
    justification=Center
    text=Owned
    textTransformPolicy=ToUpper
    font {
     bForceMonospaced=false
     bMaterialIsStencil=false
     fontMaterial=/FortUILibrary/Systems/ItemShopTile/Materials/M_UI_ItemShop_Text_Price_Owned
     fontObject=&Burbank
     letterSpacing=0
     monospacedWidth=1
     size=13
     skewAmount=0
     typefaceFontName=BurbankRegular-Black
     outlineSettings {
      bApplyOutlineToDropShadows=false
      bMiteredCorners=false
      bSeparateFillAlpha=false
      outlineMaterial=null
      outlineSize=0
      outlineColor=<rgba 0 0 0 1>
     }
    }
    renderTransform {angle=0, scale=<1 1>, shear=<-8 0>, translation=<0 0>}
    shadowOffset=<0 0>
   }
   slot {class=OverlaySlot, properties {horizontalAlignment=HAlign_Right, verticalAlignment=VAlign_Center, padding=<ltrb 0 0 10 0>}}
  }
  Spacer_1 {widgetClassPath=&Spacer, parent=WithUI, widget=Spacer_1, properties {size=<32 1>}, slot {class=HorizontalBoxSlot}}
  Normal_UI {
   widgetClassPath=&SizeBox
   parent=WithUI
   widget=Normal_UI
   properties {bOverride_HeightOverride=true, bOverride_WidthOverride=true, heightOverride=720, widthOverride=416}
   slot {class=HorizontalBoxSlot}
  }
  NormalOverlay {widgetClassPath=&Overlay, parent=Normal_UI, widget=NormalOverlay, slot {class=SizeBoxSlot}}
  Image_Offer_Normal_UI {
   widgetClassPath=&EpicCMSImage
   bIsVariable=true
   parent=NormalOverlay
   widget=Image_Offer_Normal_UI
   properties {brush {
     bIsDynamicallyLoaded=false
     drawAs=Image
     imageType=NoImage
     mirroring=NoMirror
     resourceName=None
     resourceObject=&MI_EID_GasStation
     tiling=NoTile
     imageSize=<512 512>
     margin=<ltrb 0 0 0 0>
     outlineSettings {
      bUseBrushTransparency=false
      roundingType=HalfHeightRadius
      width=0
      color {colorUseRule=UseColor_Specified, specifiedColor=<rgba 0 0 0 0>}
      cornerRadii=<0 0 0 0>
     }
     tintColor {colorUseRule=UseColor_Specified, specifiedColor=<rgba 1 1 1 1>}
     uVRegion {bIsValid=false, max=<0 0>, min=<0 0>}
    }}
   slot {class=OverlaySlot, properties {horizontalAlignment=HAlign_Fill, verticalAlignment=VAlign_Fill}}
  }
  UI_Layer_Normal {
   widgetClassPath=&VerticalBox
   parent=NormalOverlay
   widget=UI_Layer_Normal
   slot {class=OverlaySlot, properties {horizontalAlignment=HAlign_Fill, verticalAlignment=VAlign_Bottom}}
  }
  TagRotator_ItemTags_1 {
   widgetClassPath=&FortTagRotator
   bIsVariable=true
   parent=UI_Layer_Normal
   widget=TagRotator_ItemTags_1
   properties {timeBetweenTitleChanges=5}
   slot {class=VerticalBoxSlot, properties {horizontalAlignment=HAlign_Right, verticalAlignment=VAlign_Bottom, padding=<ltrb 0 0 12 0>}}
  }
  OfferIdentifier_1 {
   widgetClassPath=&Overlay
   bIsVariable=true
   parent=UI_Layer_Normal
   widget=OfferIdentifier_1
   slot {class=VerticalBoxSlot, properties {padding=<ltrb 0 12 0 0>}}
  }
  UI_Panel_Backing_SizeBox_1 {
   widgetClassPath=&SizeBox
   parent=OfferIdentifier_1
   widget=UI_Panel_Backing_SizeBox_1
   properties {bOverride_HeightOverride=true, heightOverride=108}
   slot {class=OverlaySlot, properties {horizontalAlignment=HAlign_Fill, verticalAlignment=VAlign_Fill}}
  }
  PanelBacking_1 {
   widgetClassPath=&Image
   bIsVariable=true
   parent=UI_Panel_Backing_SizeBox_1
   widget=PanelBacking_1
   properties {brush {
     bIsDynamicallyLoaded=false
     drawAs=Image
     imageType=NoImage
     mirroring=NoMirror
     resourceName=None
     resourceObject=&M_UI_OfferTile_UI_Layer
     tiling=NoTile
     imageSize=<32 32>
     margin=<ltrb 0 0 0 0>
     outlineSettings {
      bUseBrushTransparency=false
      roundingType=HalfHeightRadius
      width=0
      color {colorUseRule=UseColor_Specified, specifiedColor=<rgba 0 0 0 0>}
      cornerRadii=<0 0 0 0>
     }
     tintColor {colorUseRule=UseColor_Specified, specifiedColor=<rgba 1 1 1 1>}
     uVRegion {bIsValid=false, max=<0 0>, min=<0 0>}
    }}
   slot {class=SizeBoxSlot}
  }
  Text+Price_1 {
   widgetClassPath=&VerticalBox
   parent=OfferIdentifier_1
   widget=Text+Price_1
   slot {class=OverlaySlot, properties {horizontalAlignment=HAlign_Fill, verticalAlignment=VAlign_Fill}}
  }
  Text_1 {
   widgetClassPath=&VerticalBox
   parent=Text+Price_1
   widget=Text_1
   slot {class=VerticalBoxSlot, properties {verticalAlignment=VAlign_Center, padding=<ltrb 20 10 20 0>, size {sizeRule=Fill, value=1}}}
  }
  Text_Name_1 {
   widgetClassPath=&CommonTextBlock
   bIsVariable=true
   parent=Text_1
   widget=Text_Name_1
   properties {
    justification=Center
    scrollStyle=&TextScrollStyle-Base_C
    text=Asset Name
    textTransformPolicy=ToUpper
    font {
     bForceMonospaced=false
     bMaterialIsStencil=false
     fontMaterial=&M_UI_ItemShop_Text_Header
     fontObject=&Burbank
     letterSpacing=0
     monospacedWidth=1
     size=22
     skewAmount=0
     typefaceFontName=BurbankRegular-Black
     outlineSettings {
      bApplyOutlineToDropShadows=false
      bMiteredCorners=false
      bSeparateFillAlpha=false
      outlineMaterial=null
      outlineSize=0
      outlineColor=<rgba 0 0 0 1>
     }
    }
    renderTransform {angle=0, scale=<1 1>, shear=<-12 0>, translation=<0 0>}
    shadowOffset=<0 0>
   }
   slot {class=VerticalBoxSlot, properties {verticalAlignment=VAlign_Center, size {sizeRule=Fill, value=1}}}
  }
  OwnedStatus_SizeBox_1 {
   widgetClassPath=&SizeBox
   parent=Text+Price_1
   widget=OwnedStatus_SizeBox_1
   properties {bOverride_HeightOverride=true, heightOverride=36}
   slot {class=VerticalBoxSlot, properties {verticalAlignment=VAlign_Bottom}}
  }
  Switcher_OwningStatus_1 {
   widgetClassPath=&CommonVisibilitySwitcher
   bIsVariable=true
   parent=OwnedStatus_SizeBox_1
   widget=Switcher_OwningStatus_1
   slot {class=SizeBoxSlot}
  }
  TileState_NotOwned_1 {
   widgetClassPath=&Overlay
   bIsVariable=true
   parent=Switcher_OwningStatus_1
   widget=TileState_NotOwned_1
   slot {class=CommonVisibilitySwitcherSlot}
  }
  Price_Backing_1 {
   widgetClassPath=&Image
   bIsVariable=true
   parent=TileState_NotOwned_1
   widget=Price_Backing_1
   properties {brush {
     bIsDynamicallyLoaded=false
     drawAs=Box
     imageType=NoImage
     mirroring=NoMirror
     resourceName=None
     resourceObject=&M_UI_OfferTile_StatusPanel
     tiling=NoTile
     imageSize=<32 32>
     margin=<ltrb 0.5 0 0.5 0>
     outlineSettings {
      bUseBrushTransparency=false
      roundingType=HalfHeightRadius
      width=0
      color {colorUseRule=UseColor_Specified, specifiedColor=<rgba 0 0 0 0>}
      cornerRadii=<0 0 0 0>
     }
     tintColor {colorUseRule=UseColor_Specified, specifiedColor=<rgba 1 1 1 1>}
     uVRegion {bIsValid=false, max=<0 0>, min=<0 0>}
    }}
   slot {class=OverlaySlot, properties {horizontalAlignment=HAlign_Fill, verticalAlignment=VAlign_Fill}}
  }
  HorizontalBox_1 {
   widgetClassPath=&HorizontalBox
   parent=TileState_NotOwned_1
   widget=HorizontalBox_1
   slot {class=OverlaySlot, properties {horizontalAlignment=HAlign_Right, verticalAlignment=VAlign_Center, padding=<ltrb 0 0 64 0>}}
  }
  Text_Price_1 {
   widgetClassPath=&CommonTextBlock
   bIsVariable=true
   parent=HorizontalBox_1
   widget=Text_Price_1
   properties {
    justification=Center
    text=2000
    textTransformPolicy=ToUpper
    font {
     bForceMonospaced=false
     bMaterialIsStencil=false
     fontMaterial=&M_UI_ItemShop_Text_Price
     fontObject=&Burbank
     letterSpacing=0
     monospacedWidth=1
     size=16
     skewAmount=0
     typefaceFontName=BurbankRegular-Black
     outlineSettings {
      bApplyOutlineToDropShadows=false
      bMiteredCorners=false
      bSeparateFillAlpha=false
      outlineMaterial=null
      outlineSize=0
      outlineColor=<rgba 0 0 0 1>
     }
    }
    renderTransform {angle=0, scale=<1 1>, shear=<-12 0>, translation=<0 0>}
    shadowOffset=<0 0>
   }
   slot {class=HorizontalBoxSlot, properties {horizontalAlignment=HAlign_Right}}
  }
  TileState_Owned_1 {
   widgetClassPath=&Overlay
   bIsVariable=true
   parent=Switcher_OwningStatus_1
   widget=TileState_Owned_1
   slot {class=CommonVisibilitySwitcherSlot}
  }
  OwnedBacking_1 {
   widgetClassPath=&Image
   bIsVariable=true
   parent=TileState_Owned_1
   widget=OwnedBacking_1
   properties {brush {
     bIsDynamicallyLoaded=false
     drawAs=Image
     imageType=NoImage
     mirroring=NoMirror
     resourceName=None
     resourceObject=&M_UI_OfferTile_StatusPanel_Owned
     tiling=NoTile
     imageSize=<32 32>
     margin=<ltrb 0 0 0 0>
     outlineSettings {
      bUseBrushTransparency=false
      roundingType=HalfHeightRadius
      width=0
      color {colorUseRule=UseColor_Specified, specifiedColor=<rgba 0 0 0 0>}
      cornerRadii=<0 0 0 0>
     }
     tintColor {colorUseRule=UseColor_Specified, specifiedColor=<rgba 1 1 1 1>}
     uVRegion {bIsValid=false, max=<0 0>, min=<0 0>}
    }}
   slot {class=OverlaySlot, properties {horizontalAlignment=HAlign_Fill}}
  }
  Text_Name_3 {
   widgetClassPath=&CommonTextBlock
   bIsVariable=true
   parent=TileState_Owned_1
   widget=Text_Name_3
   properties {
    justification=Center
    text=Owned
    textTransformPolicy=ToUpper
    font {
     bForceMonospaced=false
     bMaterialIsStencil=false
     fontMaterial=&M_UI_ItemShop_Text_Price_Owned
     fontObject=&Burbank
     letterSpacing=0
     monospacedWidth=1
     size=13
     skewAmount=0
     typefaceFontName=BurbankRegular-Black
     outlineSettings {
      bApplyOutlineToDropShadows=false
      bMiteredCorners=false
      bSeparateFillAlpha=false
      outlineMaterial=null
      outlineSize=0
      outlineColor=<rgba 0 0 0 1>
     }
    }
    renderTransform {angle=0, scale=<1 1>, shear=<-8 0>, translation=<0 0>}
    shadowOffset=<0 0>
   }
   slot {class=OverlaySlot, properties {horizontalAlignment=HAlign_Right, verticalAlignment=VAlign_Center, padding=<ltrb 0 0 10 0>}}
  }
  Spacer_2 {widgetClassPath=&Spacer, parent=WithUI, widget=Spacer_2, properties {size=<32 1>}, slot {class=HorizontalBoxSlot}}
  Small_UI {
   widgetClassPath=&SizeBox
   parent=WithUI
   widget=Small_UI
   properties {bOverride_HeightOverride=true, bOverride_WidthOverride=true, heightOverride=344, widthOverride=416}
   slot {class=HorizontalBoxSlot, properties {verticalAlignment=VAlign_Top}}
  }
  SmallOverlay {widgetClassPath=&Overlay, parent=Small_UI, widget=SmallOverlay, slot {class=SizeBoxSlot}}
  Image_Offer_Small_UI {
   widgetClassPath=&EpicCMSImage
   bIsVariable=true
   parent=SmallOverlay
   widget=Image_Offer_Small_UI
   properties {brush {
     bIsDynamicallyLoaded=false
     drawAs=Image
     imageType=NoImage
     mirroring=NoMirror
     resourceName=None
     resourceObject=&MI_EID_GasStation
     tiling=NoTile
     imageSize=<512 512>
     margin=<ltrb 0 0 0 0>
     outlineSettings {
      bUseBrushTransparency=false
      roundingType=HalfHeightRadius
      width=0
      color {colorUseRule=UseColor_Specified, specifiedColor=<rgba 0 0 0 0>}
      cornerRadii=<0 0 0 0>
     }
     tintColor {colorUseRule=UseColor_Specified, specifiedColor=<rgba 1 1 1 1>}
     uVRegion {bIsValid=false, max=<0 0>, min=<0 0>}
    }}
   slot {class=OverlaySlot, properties {horizontalAlignment=HAlign_Fill, verticalAlignment=VAlign_Fill}}
  }
  UI_Layer_Small {
   widgetClassPath=&VerticalBox
   parent=SmallOverlay
   widget=UI_Layer_Small
   slot {class=OverlaySlot, properties {horizontalAlignment=HAlign_Fill, verticalAlignment=VAlign_Bottom}}
  }
  TagRotator_ItemTags_2 {
   widgetClassPath=&FortTagRotator
   bIsVariable=true
   parent=UI_Layer_Small
   widget=TagRotator_ItemTags_2
   properties {timeBetweenTitleChanges=5}
   slot {class=VerticalBoxSlot, properties {horizontalAlignment=HAlign_Right, verticalAlignment=VAlign_Bottom, padding=<ltrb 0 0 12 0>}}
  }
  OfferIdentifier_2 {
   widgetClassPath=&Overlay
   bIsVariable=true
   parent=UI_Layer_Small
   widget=OfferIdentifier_2
   slot {class=VerticalBoxSlot, properties {padding=<ltrb 0 12 0 0>}}
  }
  UI_Panel_Backing_SizeBox_2 {
   widgetClassPath=&SizeBox
   parent=OfferIdentifier_2
   widget=UI_Panel_Backing_SizeBox_2
   properties {bOverride_HeightOverride=true, heightOverride=108}
   slot {class=OverlaySlot, properties {horizontalAlignment=HAlign_Fill, verticalAlignment=VAlign_Fill}}
  }
  PanelBacking_2 {
   widgetClassPath=&Image
   bIsVariable=true
   parent=UI_Panel_Backing_SizeBox_2
   widget=PanelBacking_2
   properties {brush {
     bIsDynamicallyLoaded=false
     drawAs=Image
     imageType=NoImage
     mirroring=NoMirror
     resourceName=None
     resourceObject=&M_UI_OfferTile_UI_Layer
     tiling=NoTile
     imageSize=<32 32>
     margin=<ltrb 0 0 0 0>
     outlineSettings {
      bUseBrushTransparency=false
      roundingType=HalfHeightRadius
      width=0
      color {colorUseRule=UseColor_Specified, specifiedColor=<rgba 0 0 0 0>}
      cornerRadii=<0 0 0 0>
     }
     tintColor {colorUseRule=UseColor_Specified, specifiedColor=<rgba 1 1 1 1>}
     uVRegion {bIsValid=false, max=<0 0>, min=<0 0>}
    }}
   slot {class=SizeBoxSlot}
  }
  Text+Price_2 {
   widgetClassPath=&VerticalBox
   parent=OfferIdentifier_2
   widget=Text+Price_2
   slot {class=OverlaySlot, properties {horizontalAlignment=HAlign_Fill, verticalAlignment=VAlign_Fill}}
  }
  Text_2 {
   widgetClassPath=&VerticalBox
   parent=Text+Price_2
   widget=Text_2
   slot {class=VerticalBoxSlot, properties {verticalAlignment=VAlign_Center, padding=<ltrb 20 10 20 0>, size {sizeRule=Fill, value=1}}}
  }
  Text_Name_4 {
   widgetClassPath=&CommonTextBlock
   bIsVariable=true
   parent=Text_2
   widget=Text_Name_4
   properties {
    justification=Center
    scrollStyle=&TextScrollStyle-Base_C
    text=Asset Name
    textTransformPolicy=ToUpper
    font {
     bForceMonospaced=false
     bMaterialIsStencil=false
     fontMaterial=&M_UI_ItemShop_Text_Header
     fontObject=&Burbank
     letterSpacing=0
     monospacedWidth=1
     size=22
     skewAmount=0
     typefaceFontName=BurbankRegular-Black
     outlineSettings {
      bApplyOutlineToDropShadows=false
      bMiteredCorners=false
      bSeparateFillAlpha=false
      outlineMaterial=null
      outlineSize=0
      outlineColor=<rgba 0 0 0 1>
     }
    }
    renderTransform {angle=0, scale=<1 1>, shear=<-12 0>, translation=<0 0>}
    shadowOffset=<0 0>
   }
   slot {class=VerticalBoxSlot, properties {verticalAlignment=VAlign_Center, size {sizeRule=Fill, value=1}}}
  }
  OwnedStatus_SizeBox_2 {
   widgetClassPath=&SizeBox
   parent=Text+Price_2
   widget=OwnedStatus_SizeBox_2
   properties {bOverride_HeightOverride=true, heightOverride=36}
   slot {class=VerticalBoxSlot, properties {verticalAlignment=VAlign_Bottom}}
  }
  Switcher_OwningStatus_2 {
   widgetClassPath=&CommonVisibilitySwitcher
   bIsVariable=true
   parent=OwnedStatus_SizeBox_2
   widget=Switcher_OwningStatus_2
   slot {class=SizeBoxSlot}
  }
  TileState_NotOwned_2 {
   widgetClassPath=&Overlay
   bIsVariable=true
   parent=Switcher_OwningStatus_2
   widget=TileState_NotOwned_2
   slot {class=CommonVisibilitySwitcherSlot}
  }
  Price_Backing_2 {
   widgetClassPath=&Image
   bIsVariable=true
   parent=TileState_NotOwned_2
   widget=Price_Backing_2
   properties {brush {
     bIsDynamicallyLoaded=false
     drawAs=Box
     imageType=NoImage
     mirroring=NoMirror
     resourceName=None
     resourceObject=&M_UI_OfferTile_StatusPanel
     tiling=NoTile
     imageSize=<32 32>
     margin=<ltrb 0.5 0 0.5 0>
     outlineSettings {
      bUseBrushTransparency=false
      roundingType=HalfHeightRadius
      width=0
      color {colorUseRule=UseColor_Specified, specifiedColor=<rgba 0 0 0 0>}
      cornerRadii=<0 0 0 0>
     }
     tintColor {colorUseRule=UseColor_Specified, specifiedColor=<rgba 1 1 1 1>}
     uVRegion {bIsValid=false, max=<0 0>, min=<0 0>}
    }}
   slot {class=OverlaySlot, properties {horizontalAlignment=HAlign_Fill, verticalAlignment=VAlign_Fill}}
  }
  HorizontalBox_2 {
   widgetClassPath=&HorizontalBox
   parent=TileState_NotOwned_2
   widget=HorizontalBox_2
   slot {class=OverlaySlot, properties {horizontalAlignment=HAlign_Right, verticalAlignment=VAlign_Center, padding=<ltrb 0 0 64 0>}}
  }
  Text_Price_2 {
   widgetClassPath=&CommonTextBlock
   bIsVariable=true
   parent=HorizontalBox_2
   widget=Text_Price_2
   properties {
    justification=Center
    text=2000
    textTransformPolicy=ToUpper
    font {
     bForceMonospaced=false
     bMaterialIsStencil=false
     fontMaterial=&M_UI_ItemShop_Text_Price
     fontObject=&Burbank
     letterSpacing=0
     monospacedWidth=1
     size=16
     skewAmount=0
     typefaceFontName=BurbankRegular-Black
     outlineSettings {
      bApplyOutlineToDropShadows=false
      bMiteredCorners=false
      bSeparateFillAlpha=false
      outlineMaterial=null
      outlineSize=0
      outlineColor=<rgba 0 0 0 1>
     }
    }
    renderTransform {angle=0, scale=<1 1>, shear=<-12 0>, translation=<0 0>}
    shadowOffset=<0 0>
   }
   slot {class=HorizontalBoxSlot, properties {horizontalAlignment=HAlign_Right}}
  }
  TileState_Owned_2 {
   widgetClassPath=&Overlay
   bIsVariable=true
   parent=Switcher_OwningStatus_2
   widget=TileState_Owned_2
   slot {class=CommonVisibilitySwitcherSlot}
  }
  OwnedBacking_2 {
   widgetClassPath=&Image
   bIsVariable=true
   parent=TileState_Owned_2
   widget=OwnedBacking_2
   properties {brush {
     bIsDynamicallyLoaded=false
     drawAs=Image
     imageType=NoImage
     mirroring=NoMirror
     resourceName=None
     resourceObject=&M_UI_OfferTile_StatusPanel_Owned
     tiling=NoTile
     imageSize=<32 32>
     margin=<ltrb 0 0 0 0>
     outlineSettings {
      bUseBrushTransparency=false
      roundingType=HalfHeightRadius
      width=0
      color {colorUseRule=UseColor_Specified, specifiedColor=<rgba 0 0 0 0>}
      cornerRadii=<0 0 0 0>
     }
     tintColor {colorUseRule=UseColor_Specified, specifiedColor=<rgba 1 1 1 1>}
     uVRegion {bIsValid=false, max=<0 0>, min=<0 0>}
    }}
   slot {class=OverlaySlot, properties {horizontalAlignment=HAlign_Fill}}
  }
  Text_Name_5 {
   widgetClassPath=&CommonTextBlock
   bIsVariable=true
   parent=TileState_Owned_2
   widget=Text_Name_5
   properties {
    justification=Center
    text=Owned
    textTransformPolicy=ToUpper
    font {
     bForceMonospaced=false
     bMaterialIsStencil=false
     fontMaterial=&M_UI_ItemShop_Text_Price_Owned
     fontObject=&Burbank
     letterSpacing=0
     monospacedWidth=1
     size=13
     skewAmount=0
     typefaceFontName=BurbankRegular-Black
     outlineSettings {
      bApplyOutlineToDropShadows=false
      bMiteredCorners=false
      bSeparateFillAlpha=false
      outlineMaterial=null
      outlineSize=0
      outlineColor=<rgba 0 0 0 1>
     }
    }
    renderTransform {angle=0, scale=<1 1>, shear=<-8 0>, translation=<0 0>}
    shadowOffset=<0 0>
   }
   slot {class=OverlaySlot, properties {horizontalAlignment=HAlign_Right, verticalAlignment=VAlign_Center, padding=<ltrb 0 0 10 0>}}
  }
  Spacer_3 {widgetClassPath=&Spacer, parent=WithUI, widget=Spacer_3, properties {size=<32 1>}, slot {class=HorizontalBoxSlot}}
 ]
}
